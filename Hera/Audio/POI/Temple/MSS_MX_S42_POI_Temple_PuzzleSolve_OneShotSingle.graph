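# Stereo one-shot MetaSound source that plays the music stinger for a puzzle solve at the Temple POI.
# A single stereo Wave Player starts on play, feeds both outputs, and signals finished when the wave ends.
# It uses the temple puzzle-solve attenuation settings and routes to the exploration music sound class.
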MetaSoundSource /Hera/Audio/POI/Temple/MSS_MX_S42_POI_Temple_PuzzleSolve_OneShotSingle
inputs [
 Sound {type=WaveAsset}
 UE.Source.OnPlay {type=Trigger}
]
interfaces [UE.OutputFormat.Stereo:1.0, UE.Source.OneShot:1.0, UE.Source:1.1]
outputs [
 "UE.OutputFormat.Stereo.Audio:0" {type=Audio}
 "UE.OutputFormat.Stereo.Audio:1" {type=Audio}
 UE.Source.OneShot.OnFinished {type=Trigger}
]
pages [Default {
  input_defaults {Sound=&/MX_S42_POI_Temple_PuzzleSolve_STNG, UE.Source.OnPlay=null}
  nodes [WavePlayer_0 = UE.Wave Player:Stereo(Wave Asset=^Sound, Play=^UE.Source.OnPlay) {also {version [1,1]}}]
  links [
   {dst="@UE.Source.OneShot.OnFinished", src=WavePlayer_0.On Finished}
   {dst="@UE.OutputFormat.Stereo.Audio:0", src=WavePlayer_0.Out Left}
   {dst="@UE.OutputFormat.Stereo.Audio:1", src=WavePlayer_0.Out Right}
  ]
 }]
properties {
 assetImportData=&self:AssetImportData
 attenuationSettings=&/ATT_S42_Temple_PuzzleSolve
 maxDistance=6000
 numChannels=2
 outputFormat=Stereo
 sampleRate=48000
 soundClassObject=/Game/Sounds/FortSoundClassesAndModes/Fort_Music_Exploration
}
