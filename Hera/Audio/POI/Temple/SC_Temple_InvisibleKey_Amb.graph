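# Sound cue for the looping ambience of the invisible key at the Temple POI.
# A looping wave player feeds one ambient wave through a Modulator left at default settings.
# It plays at reduced volume with Fortnite's default mono item ambient-loop attenuation, and uses the loot ambient-loop sound class and concurrency rule.
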SoundCue /Hera/Audio/POI/Temple/SC_Temple_InvisibleKey_Amb
root=Modulator_0
nodes [
 {id=Modulator_0, children [WavePlayer_0]}
 {id=WavePlayer_0, properties {bLooping=true, soundWaveAssetPtr=&/Temple_InvisibleKey_Amb_01}}
]
properties {
 attenuationSettings=/Game/Sounds/Attenuation/Fort_Item_Ambient_Loop_Mono_Default_Att
 maxDistance=1250
 soundClassObject=/Game/Sounds/FortSoundClassesAndModes/Fort_Loot_Ambient_Loop
 volumeMultiplier=0.5
 concurrencySet [/Game/Sounds/Concurrency_Rules/Loot_Ambient_Conc]
}
