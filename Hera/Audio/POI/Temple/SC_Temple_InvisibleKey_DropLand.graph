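# Sound cue for the Temple POI's invisible key, named as the sound of the dropped key landing.
# A Random node picks between two equally weighted landing waves and feeds a Modulator at default settings for variation.
# It borrows weapon-drop attenuation and concurrency and the loot-collect sound class, with volume slightly reduced.
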SoundCue /Hera/Audio/POI/Temple/SC_Temple_InvisibleKey_DropLand
root=Modulator_0
nodes [
 {id=Modulator_0, children [Random_0]}
 {id=Random_0, children [WavePlayer_1,WavePlayer_3], properties {weights [1,1]}}
 {id=WavePlayer_1, properties {soundWaveAssetPtr=&/Temple_InvisibleKey_DropLand_01}}
 {id=WavePlayer_3, properties {soundWaveAssetPtr=&/Temple_InvisibleKey_DropLand_02}}
]
properties {
 attenuationSettings=&Fort_Weapon_Drop_Att
 maxDistance=3000
 soundClassObject=&Fort_Loot_Collect
 volumeMultiplier=0.8
 concurrencyOverrides {resolutionRule=PreventNew}
 concurrencySet [&Weapon_Drop_StopOldest]
}
