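# Sound cue for picking up the invisible key at the Temple point of interest.
# A random node picks one of two equally weighted pickup waves, which passes through a modulator into a mixer whose second input is empty.
# The cue plays at reduced volume in the gadgets and items sound class, borrows weapon pickup attenuation, and overrides concurrency to prevent new instances once at the limit.
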SoundCue /Hera/Audio/POI/Temple/SC_Temple_InvisibleKey_Pickup
root=Mixer_0
nodes [
 {id=Mixer_0, children [Modulator_0,null], properties {inputVolume [1,0.8]}}
 {id=Modulator_0, children [Random_0]}
 {id=Random_0, children [WavePlayer_3,WavePlayer_4], properties {weights [1,1]}}
 {id=WavePlayer_3, properties {soundWaveAssetPtr=&/Temple_InvisibleKey_Pickup_01}}
 {id=WavePlayer_4, properties {soundWaveAssetPtr=&/Temple_InvisibleKey_Pickup_02}}
]
properties {
 attenuationSettings=&Fort_Weapon_Pickup_Att
 bOverrideConcurrency=true
 maxDistance=3500
 soundClassObject=&Fort_Gadgets_Items
 volumeMultiplier=0.6
 concurrencyOverrides {resolutionRule=PreventNew}
}
